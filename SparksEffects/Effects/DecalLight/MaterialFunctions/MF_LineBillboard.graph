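# Material function that turns a UV-mapped strip into a camera-facing tapered line billboard, outputting UV, WPO and Normal.
# WPO offsets vertices sideways along the cross product of camera-relative position and tangent, with width blended between tip and base, and lengthwise along the tangent.
# UV rescales the cross coordinate by width over length to follow the taper.
# A static switch selects U or V expansion; Normal derives from a remapped fractional UV for fake curved shading.
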MaterialFunction /SparksEffects/Effects/DecalLight/MaterialFunctions/MF_LineBillboard
graph {nodes [
  n1 = FunctionInput() {bUsePreviewValueAsDefault=true, inputName=WidthTip, inputType=FunctionInput_Scalar, sortPriority=7, previewValue=<40 0 0 20>}
  n2 = FunctionInput() {bUsePreviewValueAsDefault=true, inputName=WidthBase, inputType=FunctionInput_Scalar, sortPriority=7, previewValue=<20 0 0 20>}
  n3 = TextureCoordinate()
  n4 = FunctionInput(Preview=@n3) {
   bUsePreviewValueAsDefault=true
   description="Texture coordinate to use for Thickness mapping. If projecting from U, it will use the Y component of the specified UV index for the thickness blending. "
   inputName=UVs for Thickness
   inputType=FunctionInput_Vector2
   sortPriority=4
  }
  n5 = FunctionInput(Preview=StaticBool() {value=true}) {bUsePreviewValueAsDefault=true, inputName=Input Expand U or V UV Channel, inputType=FunctionInput_StaticBool, sortPriority=5}
  n6 = StaticSwitch(True=ComponentMask(None=@n4) {g=true}, False=ComponentMask(None=@n4) {r=true}, Value=@n5)
  n7 = LinearInterpolate(A=@n1, B=@n2, Alpha=@n6)
  n8 = FunctionInput(Preview=Constant() {r=100}) {bUsePreviewValueAsDefault=true, inputName=Lenght, inputType=FunctionInput_Scalar, sortPriority=6}
  n9 = FunctionInput(Preview=@n3) {
   bUsePreviewValueAsDefault=true
   description=The texture coordinate to use in the expansion of the spline. This defaults to texture coordinate 0.
   inputName=UVs for Projection
   inputType=FunctionInput_Vector2
   sortPriority=4
   previewValue=<5 0 0 1>
  }
  n10 = Add(A=Constant() {r=-0.5}, B=@n9)
  n11 = StaticSwitch(True=ComponentMask(None=@n10) {r=true}, False=ComponentMask(None=@n10) {g=true}, Value=@n5)
  n12 = Multiply(A=Divide(A=@n7, B=@n8), B=@n11)
  FunctionOutput(None=MaterialFunctionCall(X=Add(A=@n12) {constB=0.5}, Y=@n6) {materialFunction=&MakeFloat2}.Result) {outputName=UV}
  n13 = MaterialFunctionCall() {materialFunction=&TangentBasis}
  n14 = StaticSwitch(
   True=FunctionInput(Preview=@n13.Y V) {bUsePreviewValueAsDefault=true, inputName=TangentV_WS, sortPriority=3}
   False=FunctionInput(Preview=@n13.X U) {bUsePreviewValueAsDefault=true, inputName=TangentU_WS, sortPriority=2}
   Value=@n5
  )
  n15 = CrossProduct(
   A=FunctionInput(Preview=WorldPosition() {worldPositionShaderOffset=WPT_CameraRelative}.XYZ) {bUsePreviewValueAsDefault=true, inputName=Camera Relative World Position}
   B=@n14
  )
  n16 = Normalize(VectorInput=@n15)
  n17 = Multiply(A=@n7, B=@n11)
  n18 = Multiply(A=Multiply(A=@n8) {constB=-1}, B=OneMinus(None=@n6))
  FunctionOutput(None=Add(A=Multiply(A=@n16, B=@n17), B=Multiply(A=@n14, B=@n18))) {outputName=WPO}
  n19 = TextureCoordinate()
  n20 = Frac(None=ComponentMask(None=FunctionInput(Preview=@n19) {bUsePreviewValueAsDefault=true, inputName=UVs For NormalShading, inputType=FunctionInput_Vector2, sortPriority=16}) {r=true})
  n21 = AppendVector(A=AppendVector(A=ConstantBiasScale(None=@n20) {bias=-0.5, scale=2}, B=Constant()), B=Constant() {r=0.62})
  n22 = MaterialFunctionCall!(Vector1=@n15, Vector2=Multiply(A=@n14) {constB=-1}) {materialFunction=/Engine/Functions/Engine_MaterialFunctions02/Math/CreateThirdOrthogonalVector}
  n23 = MaterialFunctionCall(VectorToTransform=@n21, BasisX=@n22.Vector1, BasisY=@n22.Vector1, BasisZ=@n22.Vector1) {materialFunction=&Transform3x3Matrix}
  n24 = MaterialFunctionCall(Vector1=@n15, Vector2=Multiply(A=@n15) {constB=-1}) {materialFunction=&CreateThirdOrthogonalVector}
  n25 = MaterialFunctionCall(VectorToTransform=@n21, BasisX=@n24.Vector1, BasisY=@n24.Vector1, BasisZ=@n24.Vector1) {materialFunction=&Transform3x3Matrix}
  FunctionOutput(None=Multiply(A=StaticSwitch(True=@n23.Transformed Vector, False=@n25.Transformed Vector, Value=@n5)) {constB=-1}) {outputName=Normal}
 ]}
