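# Additive unlit material that fakes a spotlight's visible beam on a line mesh shaped by MF_LineBillboard through world position offset.
# A cone mask and falloff are computed analytically from UVs, using inner/outer angles, softness and hotspot controls.
# Color comes from custom primitive data or blackbody temperature, modulated by time of day and optional scrolling-noise flicker.
# View-angle, near-camera and depth fades soften the beam, with platform switches providing cheaper mobile variants.
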Material /SparksEffects/Effects/DecalLight/Materials/M_DecalLight_LightVolume
graph {
 properties {
  bAlwaysEvaluateWorldPositionOffset=true
  bAutomaticallySetUsageInEditor=false
  blendMode=BLEND_Additive
  maxWorldPositionOffsetDisplacement=2500
  previewMesh=/SparksEffects/Effects/DecalLight/Meshes/SM_Sparks_Spotlight_Line
  shadingModel=MSM_Unlit
  usedShadingModels=Unlit
 }
 nodes [
  n1 = NamedRerouteUsage()
  n2 = LinearInterpolate(B=NamedRerouteUsage(), Alpha=ScalarParameter() {defaultValue=1, group=Time Of Day, parameterName=TOD Influence, sliderMax=1}) {constA=1}
  Multiply(
   A=MaterialFunctionCall(Emissive=LinearInterpolate(B=@n1, Alpha=@n2) {constAlpha=1}) {materialFunction=&MF_FX_EmissiveAdjustments}.Result
   B=NamedRerouteUsage()
  )
  Multiply(A=LinearInterpolate(B=NamedRerouteUsage(), Alpha=@n2) {constAlpha=1}, B=NamedRerouteUsage())
  n3 = ScalarParameter() {defaultValue=500, desc="Near fade, use to adjust camera clipping", group=Light Beam Camera Fade, parameterName=NearFade, sortPriority=31}
  NamedRerouteDeclaration(None=@n3) {name=NearFade, nodeColor=<rgba 1 0 0.0703 1>}
  n4 = ScalarParameter() {
   defaultValue=0.75
   desc=The beam gradually fades out when front/back faced directly. Use higher values for a harsher gradient
   group=Light Beam Camera Fade
   parameterName=ViewAngleFade
   sortPriority=31
  }
  NamedRerouteDeclaration(None=@n4) {name=ViewAngleFade, nodeColor=<rgba 0 0.3906 1 1>}
  n5 = Subtract(A=NamedRerouteUsage(), B=NamedRerouteUsage())
  n6 = Divide(A=@n5, B=NamedRerouteUsage())
  n7 = OneMinus(None=Multiply(A=Saturate(None=@n6), B=Step(X=@n5)))
  n8 = NamedRerouteUsage()
  n9 = Add(A=Saturate(None=Multiply(A=@n7, B=@n8)) {desc=Inner Cone}, B=NamedRerouteUsage())
  n10 = NamedRerouteUsage()
  NamedRerouteDeclaration(None=Saturate(None=Multiply(A=@n9, B=@n10))) {name=Cone Mask, nodeColor=<rgba 0.7344 1 0 1>}
  n11 = ScalarParameter() {defaultValue=1, group=Fake Light Settings, parameterName=Opacity Mask Intensity, sliderMax=1}
  NamedRerouteDeclaration(None=Multiply(A=Saturate(None=@n11), B=NamedRerouteUsage())) {name=Opacity Mask, nodeColor=<rgba 1 0 0.4688 1>}
  n12 = OneMinus(None=Length(None=NamedRerouteUsage()) {is=class MaterialExpressionLength})
  n13 = Saturate(None=Divide(A=@n12, B=NamedRerouteUsage()) {constB=0.95})
  n14 = Power(Base=@n13, Exp=NamedRerouteUsage()) {constExponent=4.5}
  n15 = SmoothStep(Value=@n12) {constMax=0.2}
  n16 = ComponentMask(None=NamedRerouteUsage()) {r=true}
  n17 = SquareRoot(None=Sine(None=Multiply(A=@n16) {constB=0.5}))
  NamedRerouteDeclaration(None=Multiply(A=LinearInterpolate(A=@n13, B=@n14, Alpha=@n15), B=Saturate(None=@n17))) {name=Cone Falloff, nodeColor=<rgba 0 1 0.1328 1>}
  n18 = Clamp(None=NamedRerouteUsage()) {maxDefault=0.125}
  n19 = Multiply(A=Tangent!(None=Add(A=@n18) {constB=0.04}), B=NamedRerouteUsage()) {constB=2}
  n20 = Subtract(A=NamedRerouteUsage()) {constB=0.125}
  n21 = Multiply(A=Clamp(None=Divide(A=@n20) {constB=0.25})) {constB=1.5}
  n22 = Subtract(A=NamedRerouteUsage()) {constB=0.195}
  n23 = LinearInterpolate(Alpha=Clamp(None=Divide(A=@n22) {constB=0.25})) {constA=1, constB=0}
  n24 = MaterialFunctionCall!(
   Camera Relative World Position
   TangentU_WS
   TangentV_WS
   UVs for Thickness
   UVs for Projection
   Input Expand U or V UV Channel
   Lenght=NamedRerouteUsage()
   WidthBase=Add(A=Multiply(A=@n19, B=@n21)) {constB=100}
   WidthTip=Add(A=Multiply(A=@n19, B=@n23)) {constB=50}
   UVs For NormalShading
  ) {materialFunction=/SparksEffects/Effects/DecalLight/MaterialFunctions/MF_LineBillboard}
  NamedRerouteDeclaration(None=@n24.WPO) {name=Final WPO, nodeColor=<rgba 1 0.1172 0 1>}
  NamedRerouteDeclaration(None=@n24.UV) {name=UV, nodeColor=<rgba 0.0781 0 1 1>}
  n25 = ScalarParameter() {defaultValue=512, desc=Lenght of the light beam, group=Light Beam Settings, parameterName=Length, sliderMax=8000, sortPriority=0}
  NamedRerouteDeclaration(None=@n25) {name=BeamLength, nodeColor=<rgba 0 1 0.0648 1>}
  n26 = NamedRerouteDeclaration(None=VertexInterpolator(VS=NamedRerouteUsage()).PS) {name=Raw UV, nodeColor=<rgba 0 1 0.625 1>}
  n27 = Constant2Vector() {g=-1, r=-0.5}
  n28 = AppendVector(A=NamedRerouteDeclaration(None=Add(A=@n26, B=@n27) {constB=-0.5}) {name=Light UVs, nodeColor=<rgba 0.0312 1 0 1>}, B=NamedRerouteUsage())
  NamedRerouteDeclaration(None=DotProduct(A=Normalize(VectorInput=@n28), B=NamedRerouteUsage())) {name=CosTheta, nodeColor=<rgba 0 1 0.6953 1>}
  n29 = ScalarParameter() {defaultValue=80, group=Fake Light Settings, parameterName=Outer Cone Angle, primitiveDataIndex=1, sliderMax=90, sliderMin=1, sortPriority=1}
  n30 = Divide() {constB=360}
  n31 = Multiply(A=Round(None=@n29), B=@n30) {constB=0.0028, desc=Degrees to turns}
  n32 = ScalarParameter() {defaultValue=30, group=Fake Light Settings, parameterName=Inner Cone Angle, sliderMax=85, sliderMin=1, sortPriority=0}
  n33 = Multiply(A=Round(None=@n32), B=@n30) {constB=0.0028}
  NamedRerouteDeclaration(None=Max(A=@n31, B=@n33)) {name=Turns, nodeColor=<rgba 1 0.8672 0 1>}
  NamedRerouteDeclaration(None=Min(A=Cosine(None=@n31), B=NamedRerouteUsage())) {name=Cos Outer, nodeColor=<rgba 1 0.9844 0 1>}
  NamedRerouteDeclaration(None=Cosine(None=@n33)) {name=Cos Inner, nodeColor=<rgba 0.6641 0 1 1>}
  n34 = NamedRerouteUsage()
  n35 = NamedRerouteUsage()
  n36 = NamedRerouteUsage()
  NamedRerouteDeclaration(None=Saturate(None=SmoothStep(Min=@n34, Max=@n35, Value=@n36))) {name=Cone Base, nodeColor=<rgba 0 1 0.2969 1>}
  n37 = VectorParameter() {
   bUseCustomPrimitiveData=true
   group=Light Beam Settings
   parameterName=Decal Color
   sortPriority=-2
   defaultValue=<rgba 0.3789 0.5012 0.8385 1>
  }
  n38 = Min(A=@n37.R) {constB=12000}
  n39 = Normalize(VectorInput=BlackBody!(Temp=Max(A=@n38) {constB=1700}))
  n40 = ScalarParameter() {defaultValue=1, group=Fake Light Settings, parameterName=Use Temperature, primitiveDataIndex=3, sliderMax=1, sortPriority=3}
  n41 = Saturate(None=@n40)
  NamedRerouteDeclaration(None=Multiply(A=LinearInterpolate(A=@n37.RGB, B=@n39, Alpha=@n41), B=@n37.A)) {name=Light Color, nodeColor=<rgba 0.9453 0 1 1>}
  n42 = ScalarParameter() {defaultValue=0.5, group=Fake Light Settings, parameterName=Inner Cone Softness, primitiveDataIndex=5, sliderMax=1, sortPriority=5}
  n43 = Constant2Vector() {g=0.1, r=0}
  NamedRerouteDeclaration(None=MaterialFunctionCall(Input=Saturate(None=@n42), Target Low=@n43.R, Target High=@n43.R) {materialFunction=&RemapValueRange}.Result) {name=Inner Cone Delta, nodeColor=<rgba 1 0.6797 0 1>}
  n44 = ScalarParameter() {defaultValue=0.2, group=Fake Light Settings, parameterName=Cone Origin Softness, primitiveDataIndex=6, sliderMax=1, sortPriority=6}
  n45 = Constant2Vector() {g=0.1}
  NamedRerouteDeclaration(None=MaterialFunctionCall(Input=Saturate(None=@n44), Target Low=@n45.R, Target High=@n45.R) {materialFunction=&RemapValueRange}.Result) {name=Cone Depth, nodeColor=<rgba 0 1 0.4141 1>}
  n46 = CollectionParameter() {
   collection=&FortniteMaterialParameters
   parameterName=FortniteActiveTimeOfDay
  }
  n47 = MaterialFunctionCall(Float4=@n46) {materialFunction=&BreakOutFloat4Components}
  n48 = OneMinus(None=@n47.A)
  n49 = Add(A=Multiply(A=@n47.R) {constB=0.25}, B=Multiply(A=@n47.B) {constB=0.25})
  n50 = Saturate(None=ScalarParameter() {defaultValue=1, group=Time Of Day, parameterName=Gradual Light Transition, sliderMax=1, sortPriority=0})
  NamedRerouteDeclaration(None=LinearInterpolate(A=Step(Y=@n48) {constX=0.01, constY=0.01}, B=Add(A=@n49, B=@n47.A), Alpha=Round(None=@n50))) {name=TOD, nodeColor=<rgba 0 1 0.1797 1>}
  n51 = ScalarParameter() {defaultValue=0.5, group=Fake Light Settings, parameterName=Inner Cone Weight, primitiveDataIndex=4, sliderMax=1, sortPriority=4}
  n52 = Constant2Vector() {g=0.25}
  NamedRerouteDeclaration(None=MaterialFunctionCall(Input=Saturate(None=@n51), Target Low=@n52.R, Target High=@n52.R) {materialFunction=&RemapValueRange}.Result) {name=Inner Cone Weight, nodeColor=<rgba 1 0.8672 0 1>}
  n53 = ScalarParameter() {defaultValue=2.5, group=Fake Light Settings, parameterName=Hotspot Intensity, primitiveDataIndex=7, sliderMax=1, sortPriority=7}
  n54 = Constant2Vector() {g=0.9, r=1}
  NamedRerouteDeclaration(None=MaterialFunctionCall(Input=Saturate(None=@n53), Target Low=@n54.R, Target High=@n54.R) {materialFunction=&RemapValueRange}.Result) {name=Hotspot, nodeColor=<rgba 0 1 0.6484 1>}
  n55 = ScalarParameter() {defaultValue=3.5, group=Fake Light Settings, parameterName=Light Falloff Exponent, primitiveDataIndex=2, sliderMax=50, sortPriority=2}
  NamedRerouteDeclaration(None=Multiply(A=@n55) {constB=0.3828}) {name=Light Falloff Exponent, nodeColor=<rgba 1 0.6094 0 1>}
  n56 = VectorParameter() {
   bUseCustomPrimitiveData=true
   group=Light Beam Settings
   parameterName=Flicker Position
   primitiveDataIndex=4
   channelNames {a=-, b=Z, g=Y, r=X}
  }
  n57 = Divide(
   A=ComponentMask(None=@n56) {g=true, r=true}
   B=ScalarParameter() {defaultValue=50, group=Flickering, parameterName=Flicker UV Divide, sortPriority=34}
  ) {constB=100}
  n58 = Multiply(A=Time(), B=ScalarParameter() {defaultValue=1, group=Flickering, parameterName=Flicker Speed, sortPriority=35}) {constB=2}
  n59 = MaterialFunctionCall(
   In=TextureSample(UVs=Add(A=@n57, B=@n58)) {samplerType=SAMPLERTYPE_LinearGrayscale, texture=&T_SimplePerlin_Greyscale}.R
   Contrast=ScalarParameter() {group=Flickering, parameterName=Flicker Contrast, sortPriority=36}
  ) {materialFunction=&CheapContrast}
  NamedRerouteDeclaration(None=StaticSwitchParameter(True=Saturate(None=@n59.Result), False=Constant() {r=1}) {group=Flickering, parameterName=Flickering, sortPriority=33}) {name=Flickering, nodeColor=<rgba 0.9219 1 0 1>}
  n60 = NamedRerouteDeclaration(None=Multiply(A=NamedRerouteUsage(), B=NamedRerouteUsage())) {name=FinalColor, nodeColor=<rgba 0 0.7913 1 1>}
  n61 = Desaturation(
   None=Multiply(A=NamedRerouteUsage(), B=ScalarParameter() {defaultValue=3, group=Fake Light Settings, parameterName=Mobile Color Multiplier})
   Fraction=Constant() {r=-1}
  )
  NamedRerouteDeclaration(None=MaterialFunctionCall!("Xb1,PS4"=@n60, "PC-High&&Epic,XSX,PS5"=@n60, "PC-Perf,Android"=@n61, Switch=@n61, "PC-Low(NotPerf)&&Medium"=@n60) {materialFunction=/Game/Effects/Niagara/MasterMaterials/Functions/Common/PlatformSwitch_3D}.Result) {name=Platform Adjusted Final Color, nodeColor=<rgba 0 1 0.4397 1>}
  n62 = Multiply(A=NamedRerouteUsage())
  n63 = NamedRerouteUsage()
  n64 = NamedRerouteUsage()
  n65 = Constant() {r=1}
  n66 = Saturate(None=Multiply(A=Multiply(A=@n63) {constB=0.5}, B=Desaturation(None=@n64, Fraction=@n65)))
  NamedRerouteDeclaration(None=MaterialFunctionCall("Xb1,PS4"=@n62, "PC-High&&Epic,XSX,PS5"=@n62, "PC-Perf,Android"=@n66, Switch=@n66, "PC-Low(NotPerf)&&Medium"=@n62) {materialFunction=&PlatformSwitch_1D}.Result) {name=Platform Adjusted Opacity Mask, nodeColor=<rgba 1 0.117 0 1>}
  n67 = DepthFade()
  n68 = Constant() {r=1}
  n69 = ShadingPathSwitch(Default=QualitySwitch(Default=Multiply(A=@n68, B=@n67), Low=@n68), Mobile=@n68)
  n70 = MaterialFunctionCall(Distance=NamedRerouteUsage()) {materialFunction=&Near_Camera_Fade}
  n71 = Constant3Vector() {constant=<rgba 1 0 0 0>}
  n72 = DotProduct(A=CameraVectorWS(), B=Normalize(VectorInput=Transform(None=@n71) {transformSourceType=TRANSFORMSOURCE_Local}))
  n73 = Power(Base=OneMinus(None=Abs(None=@n72)), Exp=NamedRerouteUsage())
  NamedRerouteDeclaration(None=Multiply(A=Multiply(A=@n69, B=@n70.Result), B=Saturate(None=@n73))) {name="View, Depth and Near Fade", nodeColor=<rgba 1 0 0.3984 1>}
  NamedRerouteDeclaration(None=Constant3Vector() {constant=<rgba 0 -1 0 0>}) {name=Fake Light Dir, nodeColor=<rgba 0 1 0.4609 1>}
  NamedRerouteUsage()
 ]
}
